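# Masked Nanite material for the sentry turret arm that animates it rising and swinging into place through world position offset.
# One Animation value, from a parameter or custom primitive data, drives two overlapping inverse-lerp windows: a local Z slide and a pivot rotation masked by red vertex color.
# Pixels below a local ZHeight are clipped, and the attributes pass through MF_CreativeMoveTool.
# The rotated-normal blend and time-based smoothstep branches are left unconnected.
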Material /Game/Athena/Items/EnvironmentalItems/Sentry/Turret/M_TurretArm
graph {
 properties {bUseMaterialAttributes=true, bUsedWithNanite=true, blendMode=BLEND_Masked, pixelDepthOffsetMode=PDOM_Legacy, usedShadingModels=Default Lit}
 nodes [
  n1 = Normalize(VectorInput=Transform(None=Constant3Vector() {constant=<rgba 0 1 0 0>}) {transformSourceType=TRANSFORMSOURCE_Local})
  n2 = StaticSwitchParameter(
   True=ScalarParameter() {bUseCustomPrimitiveData=true, defaultValue=1, parameterName=Animation_PrimitiveData, primitiveDataIndex=6}
   False=ScalarParameter() {defaultValue=1, parameterName=Animation}
  ) {parameterName=UsePrimitiveData}
  n3 = Multiply(
   A=MaterialFunctionCall!(
    min=ScalarParameter() {defaultValue=0.8, parameterName=AngleEnd}
    max=ScalarParameter() {defaultValue=0.4, parameterName=AngleStart}
    value=@n2
   ) {materialFunction=/Game/Effects/Fort_Effects/Materials/_Functions/InverseLerp_MF}.Clamped
   B=ScalarParameter() {defaultValue=-0.2, parameterName=ArmRotationAngle}
  )
  n4 = TransformPosition(None=VectorParameter() {parameterName=ArmPivot, defaultValue=<rgba 23.327 0 57.296 0>}.RGB) {transformType=TRANSFORMPOSSOURCE_World}
  LinearInterpolate(
   A=Transform(None=VertexNormalWS()) {transformSourceType=TRANSFORMSOURCE_World, transformType=TRANSFORM_Tangent}
   B=MaterialFunctionCall(Rotation Axis=@n1, Rotation Angle=@n3, Normals Pivot=@n4) {materialFunction=&FixRotateAboutAxisNormals}.New Tangent Space Vertex Normal
   Alpha=VertexColor().R
  )
  n5 = TextureSample() {samplerType=SAMPLERTYPE_Masks, texture=/Game/Creative/Sets/PropSets/Playgrounds/Textures/T_Turret_Gun_S}
  n6 = ScalarParameter() {defaultValue=-5, parameterName=ZHeight}
  n7 = ComponentMask(None=TransformPosition(None=WorldPosition().XYZ) {transformSourceType=TRANSFORMPOSSOURCE_World}) {b=true}
  n8 = Constant()
  n9 = Constant() {r=1}
  n10 = Multiply(
   A=Transform(None=VectorParameter() {parameterName=MoveOffset, defaultValue=<rgba 0 0 -90 0>}.RGB) {transformSourceType=TRANSFORMSOURCE_Local}
   B=MaterialFunctionCall(min=ScalarParameter() {defaultValue=0.6, parameterName=MoveOffsetEnd}, max=ScalarParameter() {parameterName=MoveOffsetStart}, value=@n2) {materialFunction=&InverseLerp_MF}.Clamped
  )
  n11 = Multiply(A=RotateAboutAxis(NormalizedRotationAxis=@n1, RotationAngle=@n3, PivotPoint=@n4, Position=WorldPosition().XYZ), B=VertexColor().R)
  MaterialFunctionCall(MA In=MakeMaterialAttributes!(
    BaseColor=TextureSample() {texture=/Game/Creative/Sets/PropSets/Playgrounds/Textures/T_Turret_Gun_D}.RGB
    Metallic=@n5.G
    Specular=@n5.G
    Roughness=@n5.G
    Anisotropy
    EmissiveColor
    Opacity
    OpacityMask=If(A=@n6, B=@n7, "A > B"=@n8, "A == B"=@n8, "A < B"=@n9)
    Normal=TextureSample() {samplerType=SAMPLERTYPE_Normal, texture=/Game/Creative/Sets/PropSets/Playgrounds/Textures/T_Turret_Gun_N}.RGB
    Tangent
    WorldPositionOffset=Add(A=@n10, B=@n11)
    SubsurfaceColor
    ClearCoat
    ClearCoatRoughness
    AmbientOcclusion
    Refraction
    CustomizedUVs_0
    CustomizedUVs_1
    CustomizedUVs_2
    CustomizedUVs_3
    CustomizedUVs_4
    CustomizedUVs_5
    CustomizedUVs_6
    CustomizedUVs_7
    PixelDepthOffset
    ShadingModel
    Displacement
   )) {materialFunction=&MF_CreativeMoveTool}
  n12 = Time()
  n13 = ScalarParameter() {defaultValue=0.5, parameterName=Speed}
  MaterialFunctionCall!(In=Fmod(A=Multiply(A=@n12, B=@n13), B=Constant() {r=1})) {materialFunction=/Game/Effects/Fort_Effects/Materials/_Functions/MF_Smoothstep_01}
 ]
}
